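; Looping GameplayCue on the player that drives a Niagara effect pointing to the Void demon mask teleport marker.
; On start it takes the marker from the source object, falling back to the cue location, and watches the cloak tag to toggle effect visibility.
; A repeating timer feeds the last known marker location into PortalPosition and reattaches the effect to the pawn mesh to avoid culling.
; Removal clears the timer.
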(event EventOnLoopingStart(EitherVFX-preferthisversion)
  (params (MyTarget :type Actor%20Object%20Reference) (Parameters :type Gameplay%20Cue%20Parameters%20Structure%20%28by%20ref%29) (ParticleComponents :type Array%20of%20FXSystem%20Component%20Object%20References) (AudioComponents :type Array%20of%20Audio%20Component%20Object%20References))
  (bind (_normalizedmagnitude _rawmagnitude _effectcontext _matchedtagname _originaltag _aggregatedsourcetags _aggregatedtargettags _location _normal _instigator _effectcauser _sourceobject) (Ability|GameplayCue|BreakGameplayCueParameters Parameters))
  (bind _instigator_pawn (Variables|Default|GetInstigator%20Pawn))
  (bind _teleportmarker (Variables|Default|GetTeleportMarker))
  (bind _mesh (Variables|/Script/Engine.Character|GetMesh _instigator_pawn))
  (|CallParentFunction MyTarget Parameters ParticleComponents AudioComponents)
  (bind _asbga_demon_mask_void_teleport_base (Utilities|Casting|CastToBGA_DemonMask_Void_Teleport_Base _sourceobject)
    (:then
      (Variables|Default|SetTeleportMarker _asbga_demon_mask_void_teleport_base)
      (Variables|Default|SetLastMarkerLocation (Transformation|GetActorLocation _asbga_demon_mask_void_teleport_base)))
    (:CastFailed
      (Variables|Default|SetLastMarkerLocation _location)))
  (bind _asfort_player_pawn (Utilities|Casting|CastToFortPlayerPawn MyTarget))
  (Variables|Default|SetInstigator%20Pawn _asfort_player_pawn)
  ; Turn invisible if the player goes invisible
  (Ability|Tasks|WaitGameplayTagAddToActor MyTarget (Variables|Default|GetCloakTag)
    (:then
      (Ability|Tasks|WaitGameplayTagRemoveFromActor MyTarget (Variables|Default|GetCloakTag)
        (:then
          (bind _returnvalue (Utilities|Time|SetTimerbyEvent (EventDispatchers|CreateEvent :!event Custom|UpdateTeleportMarkerPos) 0.5 true true))
          (Variables|Default|SetPositionUpdateTimer _returnvalue)
          (bind _asniagara_particle_system_component (Utilities|Casting|CastToNiagaraComponent (Utilities|Array|Get(acopy) ParticleComponents)))
          (Utilities|IsValid _teleportmarker
            (:Is%20Valid
              (bind _output_get (Variables|Default|SetLastMarkerLocation (Transformation|GetActorLocation _teleportmarker))))
            (:Is%20Not%20Valid
              _))
          (Niagara|SetNiagaraVariable(Position) _asniagara_particle_system_component "PortalPosition" _output_get)
          ;[0.4741 0.4904 0.8542 0.5] Attaching FX to Player Pawn to avoid culling issues
          (if (Utilities|IsValid _mesh :!class /Script/Engine.KismetSystemLibrary)
            (Transformation|AttachComponentToComponent _asniagara_particle_system_component _mesh "AttachSocket" "SnapToTarget" "KeepRelative" "KeepRelative" true)))
        (:Removed
          (bind _asniagara_particle_system_component_2 (Utilities|Casting|CastToNiagaraComponent (Utilities|Array|Get(acopy) ParticleComponents)))
          (Rendering|SetVisibility _asniagara_particle_system_component_2 true true))))
    (:Added
      (bind _asniagara_particle_system_component_3 (Utilities|Casting|CastToNiagaraComponent (Utilities|Array|Get(acopy) ParticleComponents)))
      (Rendering|SetVisibility _asniagara_particle_system_component_3 false true))))

(event Custom|UpdateTeleportMarkerPos
  (bind _instigator_pawn (Variables|Default|GetInstigator%20Pawn))
  (bind _teleportmarker (Variables|Default|GetTeleportMarker))
  (bind _mesh (Variables|/Script/Engine.Character|GetMesh _instigator_pawn))
  (bind _asniagara_particle_system_component (Utilities|Casting|CastToNiagaraComponent (Utilities|Array|Get(acopy) (stale EventOnLoopingStart(EitherVFX-preferthisversion) ParticleComponents))))
  (Utilities|IsValid _teleportmarker
    (:Is%20Valid
      (bind _output_get (Variables|Default|SetLastMarkerLocation (Transformation|GetActorLocation _teleportmarker))))
    (:Is%20Not%20Valid
      _))
  (Niagara|SetNiagaraVariable(Position) _asniagara_particle_system_component "PortalPosition" _output_get)
  ;[0.4741 0.4904 0.8542 0.5] Attaching FX to Player Pawn to avoid culling issues
  (if (Utilities|IsValid _mesh :!class /Script/Engine.KismetSystemLibrary)
    (Transformation|AttachComponentToComponent _asniagara_particle_system_component _mesh "AttachSocket" "SnapToTarget" "KeepRelative" "KeepRelative" true)))

(event EventOnRemoval(EitherVFX-preferthisversion)
  (params (MyTarget :type Actor%20Object%20Reference) (Parameters :type Gameplay%20Cue%20Parameters%20Structure%20%28by%20ref%29) (ParticleComponents :type Array%20of%20FXSystem%20Component%20Object%20References) (AudioComponents :type Array%20of%20Audio%20Component%20Object%20References) (BurstCameraShakeInstance :type Camera%20Shake%20Base%20Object%20Reference) (BurstDecalInstance :type Decal%20Actor%20Object%20Reference))
  (|CallParentFunction MyTarget Parameters ParticleComponents AudioComponents BurstCameraShakeInstance BurstDecalInstance)
  (Utilities|Time|ClearandInvalidateTimerbyHandle self (Variables|Default|GetPositionUpdateTimer)))

(fn ConstructionScript
  (|CallParentFunction))
