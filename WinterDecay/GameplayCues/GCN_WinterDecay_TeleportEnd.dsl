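; Burst gameplay cue that plays the end-of-teleport feedback for WinterDecay.
; After the parent's burst effects, it shows a local Gunshot-type sound indicator at the target, ranged by the first burst sound's max distance.
; It spawns a radial force physics field there only off mobile, outside PC performance mode and at higher rendering detail.
; It finally deactivates the teleport audio control bus mix, using the target's instigator.
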(fn OnBurst(EitherVFX-preferthisversion)
  (params (MyTarget :type Actor%20Object%20Reference) (Parameters :type Gameplay%20Cue%20Parameters%20Structure%20%28by%20ref%29) (ParticleComponents :type Array%20of%20FXSystem%20Component%20Object%20References) (AudioComponents :type Array%20of%20Audio%20Component%20Object%20References) (BurstCameraShakeInstance :type Camera%20Shake%20Base%20Object%20Reference) (BurstDecalInstance :type Decal%20Actor%20Object%20Reference))
  (bind _returnvalue (Transformation|GetActorLocation MyTarget))
  (bind _instigator (Variables|/Script/Engine.Actor|GetInstigator MyTarget))
  (|CallParentFunction MyTarget Parameters ParticleComponents AudioComponents BurstCameraShakeInstance BurstDecalInstance)
  (bind _asfort_player_pawn (Utilities|Casting|CastToFortPlayerPawn MyTarget)
    (:then
      _)
    (:CastFailed
      _))
  (CallFunction|PlaySoundIndicator _returnvalue _instigator)
  ;[0.1302 0 0] VFX: Radial Force
  (if (Math|Boolean|ANDBoolean (not (Fortnite|IsMobilePlatform)) (not (Class|PlatformHelpersLibrary|IsPcPerfMode 0)) (> (Rendering|GetRenderingDetailMode) 1))
    (Game|SpawnActorfromClass 0 "/Game/Effects/Niagara/Blueprints/PhsyicsFields/FN_RadialForce.FN_RadialForce_C" (Math|Transform|MakeTransform _returnvalue) "Undefined" "MultiplyWithRoot" 0 3.0 1.0 8.5e+02 -0.9 0.0 0)
    (else
      _))
  (Audio|Modulation|DeactivateControlBusMix _instigator (Variables|Default|GetCBM_Teleport))
  (return))

(fn PlaySoundIndicator
  (params (Location :type Vector) (Instigator :type Actor%20Object%20Reference))
  (bind _bursteffectdata (Variables|BurstEffects|GetBurstEffectData))
  (bind (_burstparticles _burstsounds) (Utilities|Struct|BreakFortBurstEffectData _bursteffectdata))
  (bind (_condition _attachment _soundcue) (Utilities|Struct|BreakFortGameplayCueAudioInfo (Utilities|Array|Get(aref) _burstsounds)))
  (if (> (Utilities|Array|Length _burstsounds) 0)
    (Fortnite|ShowSoundIndicatorLocal :Location Location :MaxAudibleDistance (Variables|/Script/Engine.SoundBase|GetMaxDistance _soundcue) :IndicatorType "Gunshot" :Instigator Instigator :Tint "(R=0.000000,G=0.000000,B=0.000000,A=0.000000)" :SoundBase _soundcue :AttenuationMaxRangeMultiplier 1.0 :AttenuationIntensityMultiplier 1.0) ; Indicator for fuse
  ))
